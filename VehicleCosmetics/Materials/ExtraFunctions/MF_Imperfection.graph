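# Material function for vehicle cosmetics that generates surface-imperfection roughness from scratches, clouds and grunge, output as Imperfection_MA.
# Each layer samples one channel of T_Imperfection_Noise with its own tiling, exponent and strength, packed in a vector parameter.
# The layers are summed and saturated to drive a lerp between min and max Roughness, replacing the input's roughness.
# Other connected output attributes are shown wired to the input's BaseColor rather than matching pins.
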MaterialFunction /VehicleCosmetics/Materials/ExtraFunctions/MF_Imperfection
graph {nodes [
  n1 = BreakMaterialAttributes!(Attr=FunctionInput() {inputName=MaterialAttributes, inputType=FunctionInput_MaterialAttributes})
  n2 = ScalarParameter() {defaultValue=0.2, group=Imperfection, parameterName=Imperfection_Roughness_Min}
  n3 = ScalarParameter() {defaultValue=1, group=Imperfection, parameterName=Imperfection_Roughness_Max}
  n4 = NamedRerouteUsage()
  FunctionOutput(None=MakeMaterialAttributes!(
    BaseColor=@n1.BaseColor
    Metallic=@n1.BaseColor
    Specular=@n1.BaseColor
    Roughness=LinearInterpolate(A=@n2, B=@n3, Alpha=@n4)
    Anisotropy=@n1.BaseColor
    EmissiveColor=@n1.BaseColor
    Opacity=@n1.BaseColor
    OpacityMask=@n1.BaseColor
    Normal=@n1.BaseColor
    Tangent=@n1.BaseColor
    WorldPositionOffset=@n1.BaseColor
    SubsurfaceColor=@n1.BaseColor
    ClearCoat=@n1.BaseColor
    ClearCoatRoughness=@n1.BaseColor
    AmbientOcclusion=@n1.BaseColor
    Refraction=@n1.BaseColor
    CustomizedUVs_0=@n1.BaseColor
    CustomizedUVs_1=@n1.BaseColor
    CustomizedUVs_2
    CustomizedUVs_3
    CustomizedUVs_4
    CustomizedUVs_5
    CustomizedUVs_6
    CustomizedUVs_7
    PixelDepthOffset
    ShadingModel=@n1.BaseColor
    Displacement
   )) {outputName=Imperfection_MA}
  n5 = VectorParameter() {
   group=Imperfection
   parameterName=Scratches_Parameters
   sortPriority=0
   channelNames {a=AvailableChannel, b=Scratches_Roughness_Exponent, g=Scratches_Roughness_Strength, r=Scratches_Tiling}
   defaultValue=<rgba 12 1 1 1>
  }
  n6 = TextureSample(UVs=Multiply(A=TextureCoordinate(), B=@n5.Scratches_Tiling)) {samplerSource=SSM_Wrap_WorldGroupSettings, samplerType=SAMPLERTYPE_Masks, texture=/VehicleCosmetics/SharedTextures/T_Imperfection_Noise}
  NamedRerouteDeclaration(None=Multiply(A=@n5.Scratches_Roughness_Strength, B=Power(Base=@n6.R, Exp=@n5.Scratches_Roughness_Exponent))) {name=ScratchesRough, nodeColor=<rgba 0 0.2031 1 1>}
  n7 = VectorParameter() {
   group=Imperfection
   parameterName=Clouds_Parameters
   sortPriority=1
   channelNames {a=AvailableChannel, b=Clouds_Roughness_Exponent, g=Clouds_Roughness_Strength, r=Clouds_Tiling}
   defaultValue=<rgba 2 0.25 3 1>
  }
  n8 = TextureSample(UVs=Multiply(A=TextureCoordinate(), B=@n7.Clouds_Tiling)) {samplerSource=SSM_Wrap_WorldGroupSettings, samplerType=SAMPLERTYPE_Masks, texture=&T_Imperfection_Noise}
  NamedRerouteDeclaration(None=Multiply(A=@n7.Clouds_Roughness_Strength, B=Power(Base=@n8.G, Exp=@n7.Clouds_Roughness_Exponent))) {name=CloudsRough, nodeColor=<rgba 0 0.2031 1 1>}
  n9 = VectorParameter() {
   group=Imperfection
   parameterName=Grunge_Parameters
   sortPriority=2
   channelNames {a=AvailableChannel, b=Grunge_Roughness_Exponent, g=Grunge_Roughness_Strength, r=Grunge_Tiling}
   defaultValue=<rgba 2 0.2 1 1>
  }
  n10 = TextureSample(UVs=Multiply(A=TextureCoordinate(), B=@n9.Grunge_Tiling)) {samplerSource=SSM_Wrap_WorldGroupSettings, samplerType=SAMPLERTYPE_Masks, texture=&T_Imperfection_Noise}
  NamedRerouteDeclaration(None=Multiply(A=@n9.Grunge_Roughness_Strength, B=Power(Base=@n10.B, Exp=@n9.Grunge_Roughness_Exponent))) {name=GrungeRough, nodeColor=<rgba 0 0.2031 1 1>}
  n11 = Add(A=NamedRerouteUsage(), B=NamedRerouteUsage())
  n12 = NamedRerouteUsage()
  NamedRerouteDeclaration(None=Saturate(None=Add(A=@n11, B=@n12))) {name=ImperfectionRough, nodeColor=<rgba 0 1 0.8359 1>}
 ]}
