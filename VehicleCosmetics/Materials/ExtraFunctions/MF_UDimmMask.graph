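# Material function that outputs a mask for a UDIM tile selected by the UDimm_UVLocation input.
# It shifts the texture coordinates by that location with U negated, then multiplies the result by itself minus one.
# Ceil, Saturate and OneMinus turn this into 1 inside the unit range and 0 outside, per axis without branching.
# The output multiplies the R component by itself, so only the U axis affects the mask.
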MaterialFunction /VehicleCosmetics/Materials/ExtraFunctions/MF_UDimmMask
graph {nodes [
  n1 = Constant2Vector()
  n2 = Add(
   A=TextureCoordinate()
   B=Multiply(
    A=FunctionInput(Preview=@n1) {bUsePreviewValueAsDefault=true, inputName=UDimm_UVLocation, inputType=FunctionInput_Vector2}
    B=Constant2Vector() {g=1, r=-1}
   ) {constA=-1}
  )
  n3 = Ceil(None=Multiply(A=Add(A=@n2) {constB=0}, B=Add(A=@n2) {constB=-1}))
  n4 = MaterialFunctionCall(Float2=OneMinus(None=Saturate(None=@n3))) {materialFunction=&BreakOutFloat2Components}
  FunctionOutput(None=Multiply(A=@n4.R, B=@n4.R)) {outputName=Mask, sortPriority=1}
 ]}
