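# PixelateUV_MF is a library-exposed material function that quantizes UVs into a grid of NumPixels cells to produce a pixelated look.
# It floors NumPixels to a whole number, multiplies the UV input by it, floors the result and divides by the same count, snapping to each cell's lower corner.
# UV defaults to the mesh texture coordinates; a second, incomplete Floor-Multiply-Divide chain is unconnected to the output.
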MaterialFunction /VehicleCosmetics/Materials/ExtraFunctions/PixelateUV_MF
graph {
 properties {bExposeToLibrary=true}
 nodes [
  n1 = Floor()
  Divide(A=Floor(None=Multiply(B=@n1)))
  n2 = Floor(None=FunctionInput(Preview=Constant() {r=1}) {inputName=NumPixels, inputType=FunctionInput_Scalar, sortPriority=1})
  n3 = Multiply(A=FunctionInput(Preview=TextureCoordinate()) {bUsePreviewValueAsDefault=true, inputName=UV, inputType=FunctionInput_Vector2}, B=@n2)
  FunctionOutput(None=Divide(A=Floor(None=@n3), B=@n2))
 ]
}
